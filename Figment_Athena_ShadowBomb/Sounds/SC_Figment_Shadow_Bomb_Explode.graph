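# Sound cue for the Shadow Bomb explosion, played as a ranged-weapon projectile effect.
# A single explosion wave feeds a Modulator left at default settings for slight pitch and volume variation.
# It borrows smoke grenade attenuation for spatial falloff.
# A concurrency override allows only one instance at a time, stopping the oldest when a new one starts.
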SoundCue /Figment_Athena_ShadowBomb/Sounds/SC_Figment_Shadow_Bomb_Explode
root=Modulator_0
nodes [
 {id=Modulator_0, children [WavePlayer_0]}
 {id=WavePlayer_0, properties {soundWaveAssetPtr=&/WPN_Shadow_Bomb_Expl_01}}
]
properties {
 attenuationSettings=&/ATT_Figment_Fort_Grenade_Smoke
 bOverrideConcurrency=true
 maxDistance=2097152
 soundClassObject=&Fort_WeaponRanged_Projectile_Effects
 volumeMultiplier=1
 concurrencyOverrides {maxCount=1, resolutionRule=StopOldest}
}
