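# Sound cue for picking up the Shadow Bomb grenade in the Figment Athena game mode.
# A single wave player feeds the output directly, reusing a teleporter pickup wave with no randomization or modulation.
# It plays at reduced volume through the shared weapon pickup attenuation and is assigned to the gadgets sound class.
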SoundCue /Figment_Athena_ShadowBomb/Sounds/SC_Figment_grenade_pickup_Athena
root=WavePlayer_0
nodes [{id=WavePlayer_0, properties {soundWaveAssetPtr=&/teleporter_pickup_01}}]
properties {
 attenuationSettings=&Fort_Weapon_Pickup_Att
 maxDistance=3500
 soundClassObject=&Fort_Gadgets
 volumeMultiplier=0.15
}
